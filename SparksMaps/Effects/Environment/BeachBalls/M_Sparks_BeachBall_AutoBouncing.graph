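# Masked beach ball material that animates balls bouncing on their own or floating, entirely in the shader through world position offset.
# A time-driven phase with custom primitive data controls bounce height, looping travel between points, curve-atlas squash and stretch, and alternating spin.
# A static switch selects floating mode, with noise-driven bobbing and slow rotation varied by pivot-derived ID.
# Rotations include corrected normals, and camera depth fade feeds dithered masking.
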Material /SparksMaps/Effects/Environment/BeachBalls/M_Sparks_BeachBall_AutoBouncing
graph {
 properties {
  bUsedWithMeshDeformer=true
  bUsedWithSkeletalMesh=true
  blendMode=BLEND_Masked
  pixelDepthOffsetMode=PDOM_Legacy
  usedShadingModels=Default Lit
 }
 nodes [
  n1 = ScalarParameter() {defaultValue=0.5, group=WPO_isFloating, parameterName=isFloating_BallScale}
  Add(A=@n1) {constB=-1}
  n2 = WorldPosition() {worldPositionShaderOffset=WPT_ExcludeAllShaderOffsets}
  n3 = MaterialFunctionCall() {materialFunction=&ObjectPivotPoint}
  Add(A=@n2.XYZ, B=Multiply(A=Subtract(A=@n2.XYZ, B=@n3.Object Pivot Location), B=@n1))
  MaterialFunctionCall(Alpha Threshold=MaterialFunctionCall(
    Fade Length=ScalarParameter() {defaultValue=200, group=DepthFade, parameterName="Fade Length (S)"}
    Fade Offset=ScalarParameter() {defaultValue=150, group=DepthFade, parameterName="Fade Offset (S)"}
   ) {materialFunction=&CameraDepthFade}.Result) {materialFunction=&DitherTemporalAA_Fortnite}
  n4 = ScalarParameter() {bUseCustomPrimitiveData=true, defaultValue=300, group=Bounce Ranges, parameterName=CPD_1_MaxBounceHeight, primitiveDataIndex=1}
  n5 = LinearInterpolate(A=ScalarParameter() {defaultValue=150, group=Bounce Ranges, parameterName=Bounce_HeightMin}, B=@n4, Alpha=NamedRerouteUsage()) {constA=250, constB=450}
  n6 = Sine(None=Frac(None=NamedRerouteUsage())) {desc=Bounce Sinewave, period=0.3333}
  n7 = Abs(None=@n6)
  NamedRerouteDeclaration(None=MaterialFunctionCall(B=Multiply(A=@n5, B=@n7)) {materialFunction=&AppendMany}.RGB) {name=BounceHeight, nodeColor=<rgba 0.0074 0.3524 1 1>}
  n8 = Normalize(VectorInput=Constant3Vector() {constant=<rgba 0 0 1 0>})
  n9 = MaterialFunctionCall() {materialFunction=&ObjectPivotPoint}
  n10 = Add(A=WorldPosition() {worldPositionShaderOffset=WPT_ExcludeAllShaderOffsets}.XYZ, B=NamedRerouteUsage())
  n11 = NamedRerouteUsage()
  n12 = Normalize(VectorInput=Constant3Vector() {constant=<rgba 0 0 1 0>})
  n13 = MaterialFunctionCall() {materialFunction=&ObjectPivotPoint}
  n14 = Add(A=WorldPosition() {worldPositionShaderOffset=WPT_ExcludeAllShaderOffsets}.XYZ, B=NamedRerouteUsage())
  n15 = Multiply(A=@n11) {constB=-1}
  NamedRerouteDeclaration(None=LinearInterpolate(
    A=RotateAboutAxis(NormalizedRotationAxis=@n8, RotationAngle=@n11, PivotPoint=@n9.Object Pivot Location, Position=@n10)
    B=RotateAboutAxis(NormalizedRotationAxis=@n12, RotationAngle=@n15, PivotPoint=@n13.Object Pivot Location, Position=@n14)
    Alpha=NamedRerouteUsage()
   ) {desc=DirectionSwitch}) {name=RotatedWPO, nodeColor=<rgba 0 0.5781 1 1>}
  n16 = Constant3Vector() {constant=<rgba 0 0 1 0>}
  n17 = NamedRerouteUsage()
  n18 = Multiply(
   A=LinearInterpolate(Alpha=Step(Y=@n17) {constX=0.5}) {constA=-1}
   B=Multiply(A=Time(), B=ScalarParameter() {defaultValue=0.05, group=WPO_isFloating, parameterName=Floating_RotSpeed})
  )
  NamedRerouteDeclaration(None=MaterialFunctionCall(Rotation Axis=Normalize(VectorInput=@n16), Rotation Angle=@n18, World Space Vertex Normals=VertexNormalWS()) {materialFunction=&FixRotateAboutAxisNormals}.New Tangent Space Vertex Normal) {name=isFloating_FixNormals, nodeColor=<rgba 1 0.5391 0 1>}
  n19 = MaterialFunctionCall(Color=MaterialFunctionCall() {materialFunction=&ObjectPivotPoint}.Object Pivot Location) {materialFunction=&SplitComponents}
  n20 = Add(A=Add(A=@n19.R, B=@n19.R), B=@n19.B)
  NamedRerouteDeclaration(None=Frac(None=Multiply(A=@n20) {constB=77.17})) {name=isFloating_Ball_ID, nodeColor=<rgba 1 0 0.1875 1>}
  n21 = Time()
  n22 = ScalarParameter() {defaultValue=0.075, group=Bounce Ranges, parameterName=BounceSpeed}
  NamedRerouteDeclaration(None=Add(A=Multiply(A=@n21, B=@n22), B=ScalarParameter() {bUseCustomPrimitiveData=true, group=Bounce Ranges, parameterName=CPD_0_BounceTimingOffset})) {name=BounceSpeed, nodeColor=<rgba 0.8516 0 1 1>}
  n23 = WorldPosition()
  n24 = Multiply(A=Time(), B=ScalarParameter() {defaultValue=1400, group=WPO_isFloating, parameterName=Bobbing_Speed, sortPriority=7}) {constB=250}
  n25 = NamedRerouteUsage()
  n26 = Multiply(
   A=MaterialFunctionCall(
    TextureObject=TextureObject() {samplerType=SAMPLERTYPE_Grayscale, texture=&LowResBlurredNoise}
    TextureSize=ScalarParameter() {defaultValue=8000, group=WPO_isFloating, parameterName=Noise Size, sortPriority=6}
    WorldPosition=Add(A=@n23.XYZ, B=@n24)
   ) {materialFunction=&WorldAlignedTexture}.Z Texture
   B=Multiply(
    A=ScalarParameter() {defaultValue=50, group=WPO_isFloating, parameterName=Bobbing_Intensity, sortPriority=8}
    B=LinearInterpolate(Alpha=@n25) {constA=0.5}
   )
  ) {constB=5000}
  n27 = Constant3Vector() {constant=<rgba 0 0 1 0>}
  NamedRerouteDeclaration(None=Add(A=Multiply(A=@n26, B=@n27), B=NamedRerouteUsage())) {name=WPO_FloatingBall, nodeColor=<rgba 0.4766 0 1 1>}
  n28 = NamedRerouteUsage()
  n29 = Multiply(A=LinearInterpolate(
    A=ScalarParameter() {defaultValue=1, group=Bounce Ranges, parameterName=RotationSpeedMin}
    B=ScalarParameter() {defaultValue=2, group=Bounce Ranges, parameterName=RotationSpeedMax}
    Alpha=NamedRerouteUsage()
   ) {constA=6, constB=12}) {constB=6}
  NamedRerouteDeclaration(None=Frac(None=Multiply(A=@n28, B=@n29))) {name=BallRotationSpeed, nodeColor=<rgba 1 0 0.3558 1>}
  n30 = NamedRerouteUsage()
  n31 = VertexNormalWS()
  n32 = Normalize(VectorInput=Constant3Vector() {constant=<rgba 0 0 1 0>})
  n33 = Multiply(A=NamedRerouteUsage()) {constB=-1}
  n34 = VertexNormalWS()
  NamedRerouteDeclaration(None=LinearInterpolate(
    A=MaterialFunctionCall(Rotation Axis=@n32, Rotation Angle=@n30, World Space Vertex Normals=@n31) {materialFunction=&FixRotateAboutAxisNormals}.New Tangent Space Vertex Normal
    B=MaterialFunctionCall(Rotation Axis=@n32, Rotation Angle=@n33, World Space Vertex Normals=@n34) {materialFunction=&FixRotateAboutAxisNormals}.New Tangent Space Vertex Normal
    Alpha=NamedRerouteUsage()
   ) {desc=DirectionSwitch}) {name=FixedRotatedNormals, nodeColor=<rgba 1 0 0.6328 1>}
  n35 = WorldPosition() {worldPositionShaderOffset=WPT_ExcludeAllShaderOffsets}
  n36 = MaterialFunctionCall() {materialFunction=&ObjectPivotPoint}
  n37 = NamedRerouteUsage()
  n38 = CurveAtlasRowParameter(CurveTime=Frac(None=Multiply(A=@n37) {constB=6})) {
   atlas=&/CA_Sparks_BeachBall
   curve=&/C_Sparks_BeachBall_Bounce
   desc=Squash-Stretch Curves
   parameterName=BounceCurves
  }
  n39 = Multiply(
   A=MaterialFunctionCall(R=@n38.G, G=@n38.G, B=@n38.R) {materialFunction=&AppendMany}.RGB
   B=ScalarParameter() {defaultValue=0.75, group=WPO, parameterName=Ball_Scale}
  )
  NamedRerouteDeclaration(None=Multiply(A=Subtract(A=@n35.XYZ, B=@n36.Object Pivot Location), B=Add(A=@n39) {constB=-1})) {name=BallScale, nodeColor=<rgba 1 0 0.0469 1>}
  NamedRerouteDeclaration(None=Step(X=@n6) {constX=0}) {name=BounceSwitch, nodeColor=<rgba 0 0.8828 1 1>}
  n40 = ScalarParameter() {bUseCustomPrimitiveData=true, defaultValue=2, group=Bounce Ranges, parameterName=CPD_3_BounceAreaScale, primitiveDataIndex=3}
  n41 = NamedRerouteUsage()
  n42 = Multiply(A=Frac(None=Multiply(A=@n41) {constB=2})) {constB=3}
  n43 = Saturate(None=@n42)
  n44 = VectorParameter() {group=Bounce Ranges, parameterName=BouncePosition01, defaultValue=<rgba -200 -100 0 1>}
  n45 = AppendVector(A=@n44.R, B=@n44.R)
  n46 = VectorParameter() {group=Bounce Ranges, parameterName=BouncePosition02, defaultValue=<rgba 100 -100 0 1>}
  n47 = AppendVector(A=@n46.R, B=@n46.R)
  n48 = Constant()
  n49 = ScalarParameter() {bUseCustomPrimitiveData=true, defaultValue=200, parameterName=CPD_2_3rdBounce_PosOffset, primitiveDataIndex=2}
  n50 = Add(A=@n42) {constB=-1}
  n51 = VectorParameter() {group=Bounce Ranges, parameterName=BouncePosition01, defaultValue=<rgba -200 -100 0 1>}
  n52 = LinearInterpolate(
   A=LinearInterpolate(A=LinearInterpolate(A=@n45, B=@n47, Alpha=@n43), B=AppendVector(A=@n48, B=@n49), Alpha=Saturate(None=@n50))
   B=AppendVector(A=@n51.R, B=@n51.R)
   Alpha=Saturate(None=Add(A=@n42) {constB=-2})
  )
  NamedRerouteDeclaration(None=AppendVector(A=Multiply(A=@n40, B=@n52), B=Constant())) {name=BouncePosition, nodeColor=<rgba 0 0.6719 1 1>}
  n53 = Constant3Vector() {constant=<rgba 0 0 1 0>}
  NamedRerouteDeclaration(None=RotateAboutAxis(
    NormalizedRotationAxis=Normalize(VectorInput=@n53)
    RotationAngle=@n18
    PivotPoint=MaterialFunctionCall() {materialFunction=&ObjectPivotPoint}.Object Pivot Location
    Position=@n2.XYZ
   )) {name=WPO_isFloating_Rotation, nodeColor=<rgba 0 1 0.8594 1>}
  StaticSwitch(True=NamedRerouteUsage(), False=NamedRerouteUsage(), Value=StaticBoolParameter() {group=WPO, parameterName=isFloatingBall, sortPriority=1})
  n54 = NamedRerouteUsage()
  n55 = NamedRerouteUsage()
  n56 = NamedRerouteUsage()
  n57 = NamedRerouteUsage()
  StaticSwitch(
   True=NamedRerouteUsage()
   False=Add(A=Add(A=@n54, B=@n55), B=Add(A=@n56, B=@n57))
   Value=StaticBoolParameter() {group=WPO, parameterName=isFloatingBall, sortPriority=1}
  )
  TextureSample(UVs=TextureCoordinate() {uTiling=4}) {texture=&/T_Sparks_BeachBall}
  VectorParameter() {group=Bounce Ranges, parameterName=BouncePosition03, defaultValue=<rgba 0 200 0 1>}
 ]
}
